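# Translucent parent material for the amber glass on the GrindRex backpack, built on the shared MF_Glass function.
# Fakes refraction by offsetting SceneColor lookups with a normal map, blending the result with SimpleSlurp for base color and emissive at full opacity.
# Named reroutes provide inverse-Fresnel, sun scattering, depth fade and a bump-offset amber mask.
# A static switch picks between a VertexHide world-position collapse and shared character opacity.
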Material /BRCosmetics/Weapons/FORT_Melee/Pickaxe_GrindRex/Materials/M_GrindRex_Backpack_GlassAmbar_Parent
graph {
 properties {
  bCastDynamicShadowAsMasked=true
  bEnableResponsiveAA=true
  bScreenSpaceReflections=true
  bUseMaterialAttributes=true
  bUsedWithMeshDeformer=true
  bUsedWithMorphTargets=true
  bUsedWithSkeletalMesh=true
  blendMode=BLEND_Translucent
  ditherOpacityMask=true
  pixelDepthOffsetMode=PDOM_Legacy
  refractionMethod=RM_IndexOfRefraction
  translucencyDirectionalLightingIntensity=5
  translucencyLightingMode=TLM_SurfacePerPixelLighting
  usedShadingModels=Default Lit
 }
 nodes [
  n1 = TextureSample() {samplerType=SAMPLERTYPE_Masks, texture=/BRCosmetics/Weapons/FORT_Melee/Pickaxe_GrindRex/Textures/T_GrindRex_Axe_WheelMask_TA_02}
  n2 = NamedRerouteUsage()
  n3 = TextureSample(UVs=Multiply(A=TextureCoordinate()) {constB=2}) {samplerType=SAMPLERTYPE_Normal, texture=/BRCosmetics/Weapons/FORT_Melee/Pickaxe_GrindRex/Textures/T_GrindRex_Axe_Ambar_Normal}
  n4 = ScalarParameter() {defaultValue=2, parameterName=RefractionAmount}
  n5 = Power(Base=NamedRerouteUsage()) {constExponent=4}
  n6 = Multiply(A=@n3.RGB, B=Multiply(A=Divide(A=@n4) {constB=100}, B=@n5)) {constB=0.02}
  n7 = SceneColor!(UVs=Add(A=ScreenPosition().ViewportUV, B=ComponentMask(None=@n6) {g=true, r=true}))
  n8 = MaterialFunctionCall() {materialFunction=/Game/Global/Materials/MaterialGym/Materials/MaterialElements/Depth/SimpleSlurp/MF_SimpleSlurp}
  n9 = NamedRerouteUsage()
  n10 = LinearInterpolate(Alpha=NamedRerouteUsage()) {constA=0.4, constB=0.92}
  n11 = Multiply(A=LinearInterpolate(A=Multiply(A=@n2, B=@n7.RGB), B=@n8.Result, Alpha=NamedRerouteUsage()), B=Saturate(None=Max(A=@n9, B=@n10) {constB=0.5}))
  n12 = Multiply(A=Add(A=Multiply(A=@n1.G) {constB=0.3}), B=Multiply(A=@n11, B=ScalarParameter() {defaultValue=1, parameterName=Emissive_Influence}))
  Add(A=@n12, B=NamedRerouteUsage())
  Constant() {r=2}
  Constant() {r=1}
  Constant()
  LinearInterpolate(Alpha=ScalarParameter() {defaultValue=0.3, parameterName=BaseColor_Influence}) {constA=1}
  n13 = TextureCoordinate()
  n14 = ScalarParameter() {defaultValue=3, parameterName=OpacityMask_UV_Scale}
  n15 = TextureSampleParameter2D(UVs=BumpOffset(Coordinate=Multiply(A=@n13, B=@n14), Height=Constant() {r=-3})) {
   group=00_Glass
   parameterName=TechArtMask
   samplerType=SAMPLERTYPE_Masks
   sortPriority=-96
   texture=&T_BlackMask
  }
  n16 = MaterialFunctionCall!(TechArtMask=@n15.RGB) {materialFunction=/Game/Global/Materials/Character_Parent_2023/Glass_Parent_2023/MF_Glass}
  n17 = ObjectPositionWS()
  n18 = WorldPosition()
  n19 = SetMaterialAttributes!(
   MaterialAttributes=@n16.Result
   World Position Offset=Multiply(A=Subtract(A=@n17, B=@n18.XYZ), B=ScalarParameter() {group=00_Glass, parameterName=VertexHide, sortPriority=301})
  ) {attributeSetTypes [#1]}
  n20 = MaterialFunctionCall!(Material=@n16.Result) {materialFunction=/Game/Global/Materials/Character_Parent_2023/Functions/BaseFN_CharacterOpacity_2023}
  n21 = Constant() {r=1}
  n22 = SetMaterialAttributes!(
   MaterialAttributes=GetMaterialAttributes(None=StaticSwitchParameter(True=@n19, False=@n20.Result) {group=00_Glass, parameterName=UseVertexHide, sortPriority=300}) {attributeGetTypes [#2, #3, #4]}.MaterialAttributes
   Emissive Color=@n12
   Opacity=@n21
   Opacity Mask=@n21
   Normal=MaterialFunctionCall(Normal=@n3.RGB, Flatness=Constant() {r=0.65}) {materialFunction=&FlattenNormal}.Result
   Base Color=@n11
  ) {attributeSetTypes [
    #2
    #3
    #4
    #5
    #6
   ]}
  MaterialFunctionCall(In=@n22) {materialFunction=&MF_TechArtOverrides}
  n23 = VertexNormalWS()
  Multiply(A=ConstantBiasScale(None=ComponentMask(None=@n23) {b=true}), B=Constant3Vector() {constant=<rgba 0.6372 0 0.0271 1>})
  Multiply(A=ComponentMask(None=ReflectionVectorWS()) {b=true, r=true}) {constB=0.5}
  Multiply(A=@n8.Result) {constB=1.32}
  NamedRerouteDeclaration(None=@n15.G) {name=AmbarTexture, nodeColor=<rgba 1 0 0.0234 1>}
  n24 = Fresnel() {baseReflectFraction=0, exponent=1}
  NamedRerouteDeclaration(None=Power(Base=OneMinus(None=@n24), Exp=ScalarParameter() {defaultValue=5, parameterName=FresnelExponent})) {name=InFresnel, nodeColor=<rgba 0.9219 1 0 1>}
  n25 = Multiply(A=AtmosphericLightVector()) {constB=-1}
  n26 = CameraVectorWS()
  n27 = Power(
   Base=ConstantBiasScale(None=DotProduct(A=@n25, B=@n26))
   Exp=ScalarParameter() {defaultValue=2, group=SimpleSlurp, parameterName=SS_SunScatterPow, sortPriority=31}
  )
  n28 = ScalarParameter() {defaultValue=4, group=SimpleSlurp, parameterName=SS_ScatterMulitplier}
  n29 = ScalarParameter() {defaultValue=0.5, group=SimpleSlurp, parameterName=SS_ScatterStrength, sortPriority=30}
  NamedRerouteDeclaration(None=LinearInterpolate(B=Multiply(A=@n27, B=@n28), Alpha=Saturate(None=@n29)) {constA=1}) {name=Scatering, nodeColor=<rgba 0.6875 0 1 1>}
  NamedRerouteDeclaration(None=DepthFade(FadeDistance=ScalarParameter() {defaultValue=40, group=00_Glass, parameterName=FadeDistance})) {name=DepthFade, nodeColor=<rgba 1 0.4453 0 1>}
  NamedRerouteUsage()
  OneMinus(None=@n5)
 ]
}
