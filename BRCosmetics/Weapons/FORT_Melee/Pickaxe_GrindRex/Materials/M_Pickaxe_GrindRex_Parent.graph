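# Parent material for the GrindRex pickaxe cosmetic, layering a custom wheel effect between the shared 2023 character PreFX and PostFX functions.
# A SimpleSlurp depth effect, shaded by a tiled, bump-offset wheel mask texture, overrides the base surface.
# Four parameterized sphere masks plus the texture's red channel place that effect, with the blue MosquitoMask subtracted.
# A final masked blend brightens base color by up to threefold using an inverted power falloff.
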Material /BRCosmetics/Weapons/FORT_Melee/Pickaxe_GrindRex/Materials/M_Pickaxe_GrindRex_Parent
graph {
 properties {
  bAutomaticallySetUsageInEditor=false
  bUseMaterialAttributes=true
  bUsedWithMeshDeformer=true
  bUsedWithMorphTargets=true
  bUsedWithSkeletalMesh=true
  pixelDepthOffsetMode=PDOM_Legacy
  shadingModel=MSM_FromMaterialExpression
  usedShadingModels=Default Lit | Subsurface | Clear Coat | Subsurface Profile | Cloth
 }
 nodes [
  Constant3Vector() {constant=<rgba 1 0.1475 0 1>}
  Constant()
  Constant() {r=2.85}
  n1 = MaterialFunctionCall() {materialFunction=&MF_CharacterParent_2023_PreFX}
  MaterialFunctionCall(Material=NamedRerouteUsage(), M=@n1.M, Pre Skinned Local Position=@n1.M, Distance Blend=@n1.M, Fresnel=@n1.M) {materialFunction=&MF_CharacterParent_2023_PostFX}
  NamedRerouteDeclaration(None=@n1.Result) {name=InMA, nodeColor=<rgba 0 1 0.8828 1>}
  n2 = TextureCoordinate()
  n3 = TextureSample(UVs=BumpOffset(Coordinate=Multiply(A=@n2) {constB=20}, Height=Constant() {r=-4})) {
   samplerSource=SSM_Wrap_WorldGroupSettings
   samplerType=SAMPLERTYPE_Masks
   texture=/BRCosmetics/Weapons/FORT_Melee/Pickaxe_GrindRex/Textures/T_GrindRex_Axe_WheelMask_TA
  }
  n4 = MaterialFunctionCall() {materialFunction=/Game/Global/Materials/MaterialGym/Materials/MaterialElements/Depth/SimpleSlurp/MF_SimpleSlurp}
  n5 = SetMaterialAttributes!(
   MaterialAttributes=NamedRerouteUsage()
   Base Color=Multiply(A=Max(A=@n3.G) {constB=0.5}, B=@n4.Result)
   Emissive Color=Multiply(A=@n4.Result) {constB=0}
   Specular=Constant() {r=0.1}
   Roughness=Constant() {r=0.5}
  ) {attributeSetTypes [
    #1
    #2
    #3
    #4
   ]}
  n6 = TextureSample() {
   samplerSource=SSM_Wrap_WorldGroupSettings
   samplerType=SAMPLERTYPE_Masks
   texture=&T_GrindRex_Axe_WheelMask_TA
  }
  n7 = NamedRerouteUsage()
  n8 = VectorParameter() {parameterName=Skate_MaskPos01, defaultValue=<rgba 6.5 -0.4 17.8 1>}
  n9 = ScalarParameter() {defaultValue=2.85, parameterName=Skate_MaskRadius_01}
  n10 = NamedRerouteUsage()
  n11 = VectorParameter() {parameterName=Skate_MaskPos02, defaultValue=<rgba -6.9 -0.4 17.71 1>}
  n12 = ScalarParameter() {defaultValue=2.92, parameterName=Skate_MaskRadius_02}
  n13 = NamedRerouteUsage()
  n14 = Add(
   A=Add(A=SphereMask(A=@n7, B=@n8.RGB, Radius=@n9) {attenuationRadius=2.85}, B=SphereMask(A=@n10, B=@n11.RGB, Radius=@n12) {attenuationRadius=2.92})
   B=SphereMask(
    A=@n13
    B=VectorParameter() {parameterName=Skate_MaskPos03, defaultValue=<rgba -8.3 0 57 1>}.RGB
    Radius=ScalarParameter() {defaultValue=3, parameterName=Skate_MaskRadius_03}
   ) {attenuationRadius=3}
  )
  n15 = SphereMask(
   A=@n13
   B=VectorParameter() {parameterName=Skate_MaskPos04, defaultValue=<rgba 8 0 57 1>}.RGB
   Radius=ScalarParameter() {defaultValue=3, parameterName=Skate_MaskRadius_04}
  ) {attenuationRadius=3}
  n16 = Saturate(None=Add(A=@n6.R, B=Add(A=@n14, B=@n15)))
  n17 = NamedRerouteDeclaration(None=@n6.B) {name=MosquitoMask, nodeColor=<rgba 1 0 0.1875 1>}
  n18 = BlendMaterialAttributes(A=@n1.Result, B=@n5, Alpha=Saturate(None=Subtract(A=@n16, B=@n17)))
  n19 = GetMaterialAttributes(None=@n18) {attributeGetTypes [#1, #2]}
  n20 = Power(Base=OneMinus(None=NamedRerouteUsage()))
  n21 = LinearInterpolate(A=Multiply(A=@n19.Base Color) {constB=3}, B=@n19.Base Color, Alpha=Saturate(None=Add(A=@n20) {constB=0.5}))
  NamedRerouteDeclaration(None=BlendMaterialAttributes(
    A=@n18
    B=SetMaterialAttributes!(MaterialAttributes=@n19.MaterialAttributes, Base Color=@n21, Opacity Mask) {attributeSetTypes [#1, #5]}
    Alpha=NamedRerouteUsage()
   )) {name=OutMA, nodeColor=<rgba 1 0.5625 0 1>}
  NamedRerouteDeclaration(None=@n1.Fresnel) {name=InFresnel, nodeColor=<rgba 1 0.1406 0 1>}
  NamedRerouteDeclaration(None=@n1.Pre skinned Local Position) {name=PreskinnedLocalPos, nodeColor=<rgba 0 0.5547 1 1>}
 ]
}
